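# Niagara effect for the female Parfait glider cosmetic: blue electric glow, lightning and flashes around the glider.
# System values from an integrator and timer crossfade IDLE arc emitters against FORWARD bolt and HF halo, flare and spark emitters.
# Forward bolts sample sockets on the User.Weapon skeletal mesh and align to owner velocity, with mesh variants picked randomly or from arrays.
# Glows and lights use a shared flicker value; one-shot BURST emitters add flash and ring.
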NiagaraSystem /BRCosmetics/Gadgets/Assets/VinderTech_GliderChute/Glider_Parfait_Female/FX/NS_Glider_Parfait_Female
dependencies {
 usedDataInterfaces [
  &NiagaraDataInterfaceArrayFloat3
  &NiagaraDataInterfaceColorCurve
  &NiagaraDataInterfaceMeshRendererInfo
  &NiagaraDataInterfaceVector2DCurve
  &NiagaraDataInterfaceVectorCurve
 ]
 usedDynamicInputs [&RandomVector]
 usedModules [
  &SystemState
  &self:SystemSpawnScript.SystemScriptSource.SystemScriptGraph.NiagaraNodeAssignment_1.SetVariables_#1
  &FN_Timer
  &FN_Integrator
  &self:SystemSpawnScript.SystemScriptSource.SystemScriptGraph.NiagaraNodeAssignment_0.SetVariables_#2
 ]
}
emitters {
 Fissure_Glow_IDLE {
  emitterUpdateScript [
   &EmitterState(
    Inactive Response=::NewEnumerator1
    Life Cycle Mode=::NewEnumerator1
    Loop Behavior=::NewEnumerator1
    Loop Duration Mode=::NewEnumerator1
    UseLoopDelay=false
   )
   &SpawnBurst_Instantaneous(Use Spawn Probability=false, UseLoopCountLimit=false)
  ]
  particleSpawnScript [
   &InitializeParticle(
    Color=<rgba 0.07 0.225 1 1>
    Mesh Scale Mode=::NewEnumerator3
    Mesh Uniform Scale=1.01
    Position Offset=<1.3 -4.3 -26.4>
    UsePositionOffset=true
    Write Material Random=true
   )
   &CameraOffset(Camera Offset Amount=1) {enabled=false}
  ]
  particleUpdateScript [
   &ParticleState(Kill Particles When Lifetime Has Elapsed=false)
   /SetVariables(Particles.MeshColor:LinearColor=&MakeLinearColorFromVectorAndFloat(
     "Vector (RGB)"=&Multiply_VectorByFloat(
      Vector=&MakeVectorFromLinearColorRGB(Color=$Particles.Color)
      Float=&Lerp_Float(A=11, B=50, Alpha=$System.FN_Integrator001.Value)
     )
     "Float (Alpha)"=&Multiply_Float(
      A=&Lerp_Float(A=0.5, B=1, Alpha=$System.FN_Integrator001.Value)
      B=&Lerp_Float(A=0.4, B=1, Alpha=$System.Flicker)
     )
    )) {isSetParameters=true}
  ]
  rendererClasses [&NiagaraMeshRendererProperties]
  renderers [{rendererClass=&NiagaraMeshRendererProperties, rendererIndex=0}]
 }
 Flash_Flare_BURST {
  emitterUpdateScript [
   &EmitterState(Life Cycle Mode=::NewEnumerator1, Loop Behavior=::NewEnumerator1, Loop Delay=0.1, Loop Duration=1, UseLoopDelay=true)
   &SpawnBurst_Instantaneous(Spawn Count=1, Use Spawn Probability=false)
  ]
  particleSpawnScript [
   /Niagara/Modules/Spawn/Initialization/InitializeParticle(
    Color=<rgba 0.0697 0.4721 1 1>
    Lifetime=0.25
    Module.Write Color=true
    Module.Write Lifetime=true
    Module.Write Mass=true
    Module.Write Material Random=true
    Module.Write Position=false
    Module.Write Scale=false
    Module.Write SpriteRotation=false
    Module.Write SpriteSize=true
    Position=&SimulationPosition()
    Sprite Size=<440 250>
   )
   &CameraOffset(Camera Offset Amount=25)
   &DynamicMaterialParameters(
    Index 0 Param 1=0.2
    Index 0 Param 2=1
    Module.Param0WriteEnabled=true
    Module.Param1WriteEnabled=true
    Module.Param2WriteEnabled=true
    Module.Param3WriteEnabled=true
   )
   &SystemLocation(MaskSpawnGroup=false, Offset=$System.HitOffset)
  ]
  particleUpdateScript [
   &ParticleState()
   &ScaleColor(
    Scale Alpha=&Lerp_Float(A=1, B=0, Alpha=$Particles.NormalizedAge)
    Scale RGB=&VectorFromFloat(Value=5)
    ScaleA=true
    ScaleRGB=true
   )
   &FN_Light_v2(Light Exponent=3, Light Radius=150, LightEnabled=true, Scale Color=1)
  ]
  rendererClasses [&NiagaraSpriteRendererProperties, &NiagaraLightRendererProperties]
  renderers [
   {rendererClass=&NiagaraSpriteRendererProperties}
   {rendererClass=&NiagaraLightRendererProperties, rendererIndex=1}
  ]
 }
 Flash_Flare_HF {
  emitterUpdateScript [
   &EmitterState(Life Cycle Mode=::NewEnumerator1, Loop Duration=0, UseLoopDelay=false)
   &SpawnRate(SpawnRate=&Lerp_Float(A=1, B=1, Alpha=$System.FN_Integrator001.Value), Use Spawn Probability=false)
  ]
  particleSpawnScript [
   &InitializeParticle(
    Color=<rgba 0 0.2567 1 1>
    Lifetime=2
    Module.Write Color=true
    Module.Write Lifetime=true
    Module.Write Mass=true
    Module.Write Material Random=true
    Module.Write Position=false
    Module.Write Scale=false
    Module.Write SpriteRotation=false
    Module.Write SpriteSize=true
    Position=&SimulationPosition()
    Sprite Size=<150 200>
   )
   &CameraOffset(Camera Offset Amount=5)
   &DynamicMaterialParameters(
    Index 0 Param 1=0.2
    Module.Param0WriteEnabled=true
    Module.Param1WriteEnabled=true
    Module.Param2WriteEnabled=true
    Module.Param3WriteEnabled=true
   )
  ]
  particleUpdateScript [
   &ParticleState()
   &ScaleColor(
    Scale Alpha=&Lerp_Float(A=1, B=0, Alpha=$Particles.NormalizedAge)
    Scale RGB=&VectorFromFloat(Value=&Lerp_Float(A=5, B=11, Alpha=$System.FN_Integrator001.Value))
    ScaleA=true
    ScaleRGB=true
   )
   &ScaleSpriteSize(Uniform Curve Sprite Scale=DataInterface() {propertyValues {Curve {keys [
        {interpMode=RCIM_Cubic, tangentMode=RCTM_User, time=0, value=0}
        {interpMode=RCIM_Cubic, tangentMode=RCTM_User, time=0.5, value=1}
        {tangentMode=RCTM_User, time=1, value=0}
       ]}}})
  ]
  rendererClasses [&NiagaraSpriteRendererProperties]
  renderers [{rendererClass=&NiagaraSpriteRendererProperties}]
 }
 Glow_IDLE {
  emitterUpdateScript [
   &EmitterState(
    Inactive Response=::NewEnumerator1
    Life Cycle Mode=::NewEnumerator1
    Loop Behavior=::NewEnumerator1
    Loop Duration Mode=::NewEnumerator1
    UseLoopDelay=true
   )
   &SpawnBurst_Instantaneous(Use Spawn Probability=false, UseLoopCountLimit=false)
  ]
  particleSpawnScript [
   /Niagara/Modules/Spawn/Initialization/V2/InitializeParticle(Color=<rgba 0.07 0.225 1 1>, Sprite Size Mode=::NewEnumerator3, Uniform Sprite Size=66, UsePositionOffset=true, Write Material Random=true)
   /SetVariables(Particles.LightPosition:NiagaraPosition=<0 5 0>) {isSetParameters=true}
   &CameraOffset(Camera Offset Amount=5)
  ]
  particleUpdateScript [
   &ParticleState(Kill Particles When Lifetime Has Elapsed=false)
   /SetVariables(
    Particles.LightColor=&MakeLinearColorFromVectorAndFloat(
     "Vector (RGB)"=&Multiply_VectorByFloat(
      Vector=&MakeVectorFromLinearColorRGB(Color=$Particles.Color)
      Float=&Lerp_Float(A=25, B=30, Alpha=$System.FN_Integrator001.Value)
     )
     "Float (Alpha)"=&Multiply_Float(
      A=&Lerp_Float(A=0.5, B=1, Alpha=$System.FN_Integrator001.Value)
      B=&Lerp_Float(A=0.75, B=1, Alpha=$System.Flicker)
     )
    )
    Particles.LightExponent:NiagaraFloat=&Lerp_Float(A=1.666, B=2.333, Alpha=$System.Flicker)
    Particles.LightRadius=&Lerp_Float(A=77, B=88, Alpha=$System.Flicker)
    Particles.MeshColor:LinearColor=&MakeLinearColorFromVectorAndFloat(
     "Vector (RGB)"=&Multiply_VectorByFloat(Vector=&MakeVectorFromLinearColorRGB(Color=$Particles.Color), Float=11)
     "Float (Alpha)"=&Multiply_Float(
      A=&Lerp_Float(A=0.5, B=1, Alpha=$System.FN_Integrator001.Value)
      B=&Lerp_Float(A=0.4, B=1, Alpha=$System.Flicker)
     )
    )
    Particles.SpriteColor=&MakeLinearColorFromVectorAndFloat(
     "Vector (RGB)"=&Multiply_VectorByFloat(
      Vector=&MakeVectorFromLinearColorRGB(Color=$Particles.Color)
      Float=&FNTimeOfDayScalar(Night Scale=0.3, Morning Scale=0.2, Day Scale=0.2, Evening Scale=0.2)
     )
     "Float (Alpha)"=&Multiply_Float(
      A=&Lerp_Float(A=0.8, B=1, Alpha=$System.FN_Integrator001.Value)
      B=&Lerp_Float(A=0.8, B=1.25, Alpha=$System.Flicker)
     )
    )
    Particles.SpriteSize=&Multiply_Vector2D_ByFloat(
     Float=&Ceil_Float(Float=&Lerp_Float(A=0.5, B=1, Alpha=$System.FN_Integrator001.Value))
     Vector2D=$Particles.Initial.SpriteSize
    )
   ) {isSetParameters=true}
  ]
  rendererClasses [&NiagaraLightRendererProperties, &NiagaraSpriteRendererProperties]
  renderers [
   {rendererClass=&NiagaraLightRendererProperties, rendererIndex=0}
   {rendererClass=&NiagaraSpriteRendererProperties, rendererIndex=1}
  ]
 }
 Glow_Ring_BURST {
  emitterUpdateScript [
   &EmitterState(
    Enable Distance Culling=true
    Life Cycle Mode=::NewEnumerator1
    Loop Behavior=::NewEnumerator1
    Loop Delay=0.1
    Loop Duration=1
    Max Distance Response=::KillImmediately
    MaxDistance=3500
    Module.bUseMaxDistance=true
    Module.bUseMinDistance=false
    Scalability Mode=::NewEnumerator1
    UseLoopDelay=true
   )
   &SpawnBurst_Instantaneous(Spawn Count=1, Use Spawn Probability=false)
  ]
  particleSpawnScript [
   &InitializeParticle(
    Color=<rgba 0.4069 0.6973 1 1>
    Lifetime=0.15
    Sprite Rotation Mode=::NewEnumerator2
    Sprite Size Mode=::NewEnumerator3
    Uniform Sprite Size=200
    UsePositionOffset=false
    Write Material Random=true
   )
   &CameraOffset(Camera Offset Amount=20)
  ]
  particleUpdateScript [
   &ParticleState()
   &ScaleColor(
    Scale RGB=&Multiply_VectorByFloat(
     Vector=<1 1 1>
     Float=&FloatFromCurve(
      FloatCurve=DataInterface() {propertyValues {Curve {keys [
          {interpMode=RCIM_Cubic, tangentMode=RCTM_User, time=0, value=0}
          {interpMode=RCIM_Cubic, tangentMode=RCTM_User, time=0.2, value=0.95}
          {tangentMode=RCTM_User, time=1, value=0}
         ]}}}
      Scale Curve=0.5
     )
    )
    ScaleA=true
    ScaleRGB=true
   )
   &ScaleSpriteSize(Scale Factor=&Vector2DFromFloat(Value=&FloatFromCurve(FloatCurve=DataInterface() {propertyValues {Curve {keys [{arriveTangent=2.7513, interpMode=RCIM_Cubic, leaveTangent=2.7513, tangentMode=RCTM_User, time=0, value=0}, <at 1 1>]}}})))
  ]
  rendererClasses [&NiagaraSpriteRendererProperties]
  renderers [{rendererClass=&NiagaraSpriteRendererProperties}]
 }
 Halo_HF {
  emitterUpdateScript [
   &EmitterState(
    Enable Distance Culling=true
    Life Cycle Mode=::NewEnumerator1
    Loop Delay=0.1
    Loop Duration=0
    Max Distance Response=::KillImmediately
    MaxDistance=3500
    Module.bUseMaxDistance=true
    Module.bUseMinDistance=false
    Scalability Mode=::NewEnumerator1
    UseLoopDelay=false
   )
   &SpawnRate(SpawnRate=&Lerp_Float(A=0, B=1, Alpha=$System.FN_Integrator001.Value), Use Spawn Probability=false)
  ]
  particleSpawnScript [
   &InitializeParticle(
    Color=<rgba 0.2 1 2.5 5>
    Lifetime=2
    Sprite Rotation Mode=::NewEnumerator2
    Sprite Size Mode=::NewEnumerator3
    Uniform Sprite Size=150
    UsePositionOffset=false
    Write Material Random=true
   )
   &CameraOffset(Camera Offset Amount=20)
  ]
  particleUpdateScript [
   &ParticleState()
   &ScaleColor(
    Scale Alpha=&FloatFromCurve(
     FloatCurve=DataInterface() {propertyValues {Curve {keys [
         {interpMode=RCIM_Cubic, tangentMode=RCTM_User, time=0, value=0}
         {arriveTangent=0.3455, interpMode=RCIM_Cubic, leaveTangent=0.3455, tangentMode=RCTM_User, time=0.2, value=0.7}
         {tangentMode=RCTM_User, time=1, value=0}
        ]}}}
     Scale Curve=2
    )
    Scale RGB=&Multiply_VectorByFloat(
     Vector=<1 1 1>
     Float=&FloatFromCurve(
      FloatCurve=DataInterface() {propertyValues {Curve {keys [
          {interpMode=RCIM_Cubic, tangentMode=RCTM_User, time=0, value=0}
          {arriveTangent=-2.4993, interpMode=RCIM_Cubic, leaveTangent=-2.4993, tangentMode=RCTM_User, time=0.0729, value=0.8118}
          <at 0.2251 0.1818>
          {tangentMode=RCTM_User, time=1, value=0}
         ]}}}
      Scale Curve=0.35
     )
    )
    ScaleA=true
    ScaleRGB=true
   )
   &ScaleSpriteSize(Scale Factor=&Vector2DFromFloat(Value=&FloatFromCurve(FloatCurve=DataInterface() {propertyValues {Curve {keys [{arriveTangent=2.7513, interpMode=RCIM_Cubic, leaveTangent=2.7513, tangentMode=RCTM_User, time=0, value=0}, <at 1 1>]}}})))
  ]
  rendererClasses [&NiagaraSpriteRendererProperties]
  renderers [{rendererClass=&NiagaraSpriteRendererProperties}]
 }
 Lightning_Arc_Sprites_HF {
  emitterUpdateScript [
   &EmitterState(
    Delay First Loop Only=true
    Life Cycle Mode=::NewEnumerator1
    Loop Delay=0.5
    Loop Duration=0.5
    Recalculate Duration Each Loop=true
    UseLoopDelay=true
   )
   &SpawnRate(SpawnRate=&Lerp_Float(A=0, B=10, Alpha=$System.FN_Integrator001.Value), Use Spawn Probability=false)
  ]
  particleSpawnScript [
   &InitializeParticle(
    Lifetime Max=0.3
    Lifetime Min=0.2
    Lifetime Mode=::NewEnumerator1
    Sprite Rotation Mode=::NewEnumerator2
    Sprite Size Mode=::NewEnumerator1
    Uniform Sprite Size Max=40
    Uniform Sprite Size Min=20
    UsePositionOffset=false
    Write Material Random=true
   ) {module=InitializeParticle001}
   /SetVariables(Particles.SubImageIndex=&UniformRangedFloat(Minimum=0, Maximum=3)) {isSetParameters=true, module=SetVariables_#3}
   &CameraOffset(Camera Offset Amount=15)
   &ShapeLocation(MaskSpawnGroup=false, Override Seed=false, Sphere Radius=10, UseOwnerScale=true)
  ]
  particleUpdateScript [
   &DynamicMaterialParameters(
    Index 0 Param 1=&RandomRangeFloat(Minimum=0, Maximum=1)
    Index 0 Param 3=&FloatFromCurve(FloatCurve=DataInterface() {propertyValues {Curve {keys [
         {arriveTangent=4.5017, interpMode=RCIM_Cubic, leaveTangent=4.5017, tangentMode=RCTM_User, time=0, value=0}
         {interpMode=RCIM_Cubic, tangentMode=RCTM_User, time=0.3, value=1}
         {arriveTangent=-1.4286, interpMode=RCIM_Cubic, leaveTangent=-1.4286, time=0.7, value=1}
         {tangentMode=RCTM_User, time=1, value=0}
        ]}}})
    Module.Param0WriteEnabled=true
    Module.Param1WriteEnabled=true
    Module.Param2WriteEnabled=true
    Module.Param3WriteEnabled=false
   )
   &ParticleState()
   &SolveForcesAndVelocity(Clamp Velocity=false, Limit Acceleration=false)
   &ScaleColor(
    Scale Alpha=&FloatFromCurve(FloatCurve=DataInterface() {propertyValues {Curve {keys [<at 0.75 1 RCIM_Cubic>, {arriveTangent=-0.6255, interpMode=RCIM_Cubic, leaveTangent=-0.6255, tangentMode=RCTM_User, time=1, value=0}]}}})
    Scale RGB=&Multiply_VectorByFloat(Vector=&MakeVectorFromLinearColorRGB(Color=<rgba 0 2 15 1>), Float=5)
    ScaleA=true
    ScaleRGB=true
   ) {module=ScaleColor002}
  ]
  rendererClasses [&NiagaraSpriteRendererProperties]
  renderers [{rendererClass=&NiagaraSpriteRendererProperties}]
 }
 Lightning_Arcs_IDLE {
  emitterUpdateScript [
   &EmitterState(
    Inactive Response=::NewEnumerator1
    Life Cycle Mode=::NewEnumerator1
    Loop Delay=&UniformRangedFloat(Minimum=1.1, Maximum=1.5)
    Loop Duration=&UniformRangedFloat(Minimum=0.5, Maximum=0.8)
    Recalculate Duration Each Loop=true
    UseLoopDelay=true
   )
   &SpawnRate(SpawnRate=&Lerp_Float(A=1.5, B=0, Alpha=$System.FN_Integrator001.Value), Use Spawn Probability=false)
   &SpawnBurst_Instantaneous(
    Spawn Count=&MakeIntFromFloat(Float=&Lerp_Float(A=1, B=0, Alpha=$System.FN_Timer.Value))
    Use Spawn Probability=false
   )
   &SpawnBurst_Instantaneous(
    Spawn Count=&MakeIntFromFloat(Float=&Lerp_Float(A=1, B=0, Alpha=$System.FN_Timer.Value))
    Spawn Time=0.2
    Use Spawn Probability=false
   ) {module=SpawnBurst_Instantaneous001}
  ]
  particleSpawnScript [
   &InitializeParticle(
    Lifetime Max=0.2
    Lifetime Min=0.1
    Lifetime Mode=::NewEnumerator1
    Mesh Scale Max=<0.4 0.4 0.4>
    Mesh Scale Min=<0.1 0.1 0.1>
    Mesh Scale Mode=::NewEnumerator2
    UsePositionOffset=false
    Write Material Random=true
   )
   &CameraOffset(Camera Offset Amount=-50)
   &DynamicMaterialParameters(
    Index 0 Param 2=&UniformRangedFloat(Minimum=0.3, Maximum=0.4)
    Module.Param0WriteEnabled=false
    Module.Param1WriteEnabled=true
    Module.Param2WriteEnabled=false
    Module.Param3WriteEnabled=false
   ) {module=DynamicMaterialParameters001}
   &InitialMeshOrientation(
    Mesh Orientation Mode=::NewEnumerator2
    Orientation Vector=<0 0 1>
    Rotation=&UniformRangedVector(Minimum=<-1 -1 -1>, Maximum=<1 1 1>)
    Use Rotation Vector=true
   ) {module=InitialMeshOrientation001}
  ]
  particleUpdateScript [
   &ParticleState()
   &SolveForcesAndVelocity(Clamp Velocity=false, Limit Acceleration=false)
   &ScaleColor(
    Scale Alpha=&FloatFromCurve(FloatCurve=DataInterface() {propertyValues {Curve {keys [<at 0.15 1 RCIM_Cubic>, {arriveTangent=-1.1765, interpMode=RCIM_Cubic, leaveTangent=-1.1765, time=0.796, value=0.245}, <at 1 0>]}}})
    Scale RGB=&VectorFromFloat(Value=&FloatFromCurve(FloatCurve=DataInterface() {propertyValues {Curve {keys [<at 0 1>,<at 0.1242 0.9958 RCIM_Constant>,<at 0.1835 0 RCIM_Constant>,<at 0.284 0.984 RCIM_Constant>,<at 0.5317 0 RCIM_Constant>,<at 0.6402 0.984 RCIM_Constant>,<at 1 0>]}}}))
    ScaleA=true
    ScaleRGB=true
   ) {module=ScaleColor003}
   &DynamicMaterialParameters(
    Index 0 Param 1=&FloatFromCurve(FloatCurve=DataInterface() {propertyValues {Curve {keys [<at 0 1>,<at 0.7 0.35>,<at 1 0>]}}})
    Index 0 Param 2=0.3
    Index 0 Param 3=&FloatFromCurve(FloatCurve=DataInterface() {propertyValues {Curve {keys [<at 0.6 1>,<at 1 0>]}}}, Scale Curve=0.5)
    Index 0 Param 4=1
    Module.Param0WriteEnabled=true
    Module.Param1WriteEnabled=false
    Module.Param2WriteEnabled=true
    Module.Param3WriteEnabled=true
   )
  ]
  rendererClasses [&NiagaraMeshRendererProperties]
  renderers [{rendererClass=&NiagaraMeshRendererProperties, rendererIndex=0}]
 }
 Lightning_Big_FORWARD {
  emitterUpdateScript [
   &EmitterState(
    Inactive Response=::NewEnumerator1
    Life Cycle Mode=::NewEnumerator1
    Loop Delay=&UniformRangedFloat(Minimum=0.5, Maximum=1.5)
    Loop Duration=1
    Recalculate Duration Each Loop=true
    UseLoopDelay=true
   )
   &SpawnBurst_Instantaneous(
    Spawn Count=&MakeIntFromFloat(Float=&Lerp_Float(A=0, B=1, Alpha=$System.FN_Integrator001.Value))
    Use Spawn Probability=false
   )
   &SpawnRate(SpawnRate=&Lerp_Float(A=0, B=1, Alpha=$System.FN_Integrator001.Value), Use Spawn Probability=false)
   &SpawnBurst_Instantaneous(
    Spawn Count=&MakeIntFromFloat(Float=&Lerp_Float(A=0, B=1, Alpha=$System.FN_Integrator001.Value))
    Spawn Time=0.1
    Use Spawn Probability=false
   ) {module=SpawnBurst_Instantaneous001}
  ]
  particleSpawnScript [
   /SetVariables(Particles.MeshIndex=&RandomRangeInt(Minimum=0, Maximum=1)) {isSetParameters=true}
   &InitializeParticle(
    Lifetime=0.444
    Mesh Scale Max=<0.2 0.15 0.2>
    Mesh Scale Min=<0.1 0.1 0.15>
    Mesh Scale Mode=::NewEnumerator2
    UsePositionOffset=false
    Write Material Random=true
   ) {module=InitializeParticle001}
   &InitialMeshOrientation(
    Mesh Orientation Mode=::NewEnumerator2
    Rotation=&SelectVectorFromArray(
     Array Sampling Mode=::NewEnumerator0
     Direct Array Index=?
     Vector Selection Array=DataInterface() {propertyValues {FloatData [<-0.05 -0.1 0.05>, <-0.05 -0.1 -0.05>, <-0.05 0.1 0.05>]}}
     Random Seed=0
    )
    Rotation Coordinate Space=::NewEnumerator2
    Use Rotation Vector=true
   ) {module=InitialMeshOrientation001}
   &CameraOffset(Camera Offset Amount=50)
   /Niagara/Modules/Update/Orientation/V2/OrientMeshToVector(
    Look At Direction:Vector3f=&NormalizeVector(Vector To Normalize=$Engine.Owner.Velocity, Safe Normalize=true, Safe Fallback Threshold=0, Safe Fallback Vector=<0 0 0.001>)
    Override Quat=false
    Reference Axis Vector:Vector3f=<-1 0 0>
    Side Axis Direction:Vector3f=$Engine.Owner.SystemXAxis
    Up Axis Direction:Vector3f=$Engine.Owner.SystemZAxis
    UseOptionalSideAxis=true
    UseOptionalUpAxis=true
   ) {module=OrientMeshToVector001}
  ]
  particleUpdateScript [
   &SkeletalMeshLocation(
    Inherit Velocity=false
    Interpolate Mesh Sampling=false
    MaskSpawnGroup=false
    Mesh Sampling Type=::NewEnumerator5
    Sampled Orientation Coordinate Space=::World
    Sampled Position Offset=&SelectVectorFromArray(
     Array Sampling Mode=::NewEnumerator0
     Direct Array Index=?
     Vector Selection Array=DataInterface() {propertyValues {FloatData [<-15 0 0>, <-5 0 0>, <5 0 0>, <15 0 0>]}}
     Random Seed=0
    )
    Sampled Position Offset Coordinate Space:ENiagaraMeshTransforms=::NewEnumerator3
    Skeletal Mesh=$User.Weapon
    Socket Sampling Mode=::NewEnumerator5
   )
   &ParticleState()
   &SolveForcesAndVelocity(Clamp Velocity=false, Limit Acceleration=false)
   &DynamicMaterialParameters(
    Index 0 Param 1=&FloatFromCurve(FloatCurve=DataInterface() {propertyValues {Curve {keys [<at 0 1 RCIM_Cubic>, {arriveTangent=-1, interpMode=RCIM_Cubic, leaveTangent=-1, time=0.3, value=0.5}, <at 1 0>]}}})
    Index 0 Param 2=&FloatFromCurve(FloatCurve=DataInterface() {propertyValues {Curve {keys [
         {interpMode=RCIM_Cubic, leaveTangent=-0.9598, tangentMode=RCTM_Break, time=0, value=0}
         {arriveTangent=5, interpMode=RCIM_Cubic, leaveTangent=5, tangentMode=RCTM_Break, time=0.25, value=0.3}
         <at 0.3 1>
         {arriveTangent=-2.7768, interpMode=RCIM_Cubic, leaveTangent=-2.7768, tangentMode=RCTM_User, time=0.35, value=0.3}
         {arriveTangent=-0.3966, interpMode=RCIM_Cubic, tangentMode=RCTM_Break, time=1, value=0.05}
        ]}}})
    Index 0 Param 3=&FloatFromCurve(FloatCurve=DataInterface() {propertyValues {Curve {keys [
         <at 0.1 0.5 RCIM_Cubic>
         <at 0.296 1>
         {arriveTangent=-2.9528, interpMode=RCIM_Cubic, leaveTangent=-2.9528, time=0.35, value=0.4}
         {arriveTangent=4.0435, interpMode=RCIM_Cubic, leaveTangent=4.0435, time=0.55, value=0.25}
         <at 0.993 3>
        ]}}})
    Index 0 Param 4=1
    Module.Param0WriteEnabled=true
    Module.Param1WriteEnabled=true
    Module.Param2WriteEnabled=true
    Module.Param3WriteEnabled=true
   )
   &ScaleMeshSize(Scale Factor=/Niagara/DynamicInputs/ScaleByCurve/ScaleVectorByCurve(
     InitialValue:Vector3f=<1 1 1>
     ScaleCurve:NiagaraDataInterfaceVectorCurve=DataInterface() {propertyValues {XCurve {keys [<at 1 1>]}, YCurve {keys [<at 0.3 1.25>,<at 0.6 1.5>,<at 1 0.5>]}, ZCurve {keys [<at 0.3 1.25>,<at 0.6 1.5>,<at 1 0.5>]}}}
     CurveIndex:NiagaraFloat=$Particles.NormalizedAge
    ))
   &ScaleColor(
    Scale Alpha=&FloatFromCurve(FloatCurve=DataInterface() {propertyValues {Curve {keys [<at 0.15 1 RCIM_Cubic>, {arriveTangent=-1.1765, interpMode=RCIM_Cubic, leaveTangent=-1.1765, time=0.796, value=0.245}, <at 1 0>]}}})
    Scale RGB=&VectorFromFloat(Value=&FloatFromCurve(FloatCurve=DataInterface() {propertyValues {Curve {keys [<at 0 1>,<at 0.1242 0.9958 RCIM_Constant>,<at 0.1835 0 RCIM_Constant>,<at 0.2506 0.9839 RCIM_Constant>,<at 0.5317 0 RCIM_Constant>,<at 0.6151 1 RCIM_Constant>,<at 1 0>]}}}))
    ScaleA=true
    ScaleRGB=true
   ) {module=ScaleColor002}
   /SetVariables(Particles.light color:LinearColor=&MakeLinearColorFromVectorAndFloat(
     "Vector (RGB)"=&MakeVectorFromLinearColorRGB(Color=<rgba 1 29.7233 100 1>)
     "Float (Alpha)"=&FloatFromCurve(FloatCurve=DataInterface() {propertyValues {Curve {keys [<at 0 1>,<at 0.0918 0.159>,<at 0.1514 0.9748>,<at 0.2547 0.1132>,<at 0.3492 0.8648>,<at 0.3825 0.1223>,<at 0.4928 0.8923>,<at 0.5611 0.1407>,<at 0.6276 0.9107>,<at 0.6959 0.1407>,<at 0.7677 0.874>,<at 0.85 0.1315>,<at 0.9235 0.874>,<at 1 0>]}}})
    )) {isSetParameters=true}
   /SetVariables(Particles.LightRadius=&UniformRangedFloat(Minimum=0, Maximum=100)) {isSetParameters=true}
  ]
  rendererClasses [&NiagaraMeshRendererProperties, &NiagaraLightRendererProperties]
  renderers [
   {rendererClass=&NiagaraMeshRendererProperties, rendererIndex=0}
   {rendererClass=&NiagaraLightRendererProperties, rendererIndex=1}
  ]
 }
 Lightning_Big_IDLE001 {
  emitterUpdateScript [
   &EmitterState(
    Inactive Response=::NewEnumerator2
    Life Cycle Mode=::NewEnumerator1
    Loop Delay=&RandomRangeFloat(Minimum=0.5, Maximum=1)
    Loop Duration=&RandomRangeFloat(Minimum=0.111, Maximum=0.222)
    Recalculate Duration Each Loop=true
    UseLoopDelay=true
   )
   &SpawnBurst_Instantaneous(
    Spawn Count=&MakeIntFromFloat(Float=&Lerp_Float(A=1, B=0, Alpha=$System.FN_Timer.Value))
    Use Spawn Probability=false
   )
   &SpawnBurst_Instantaneous(
    Spawn Count=&MakeIntFromFloat(Float=&Lerp_Float(A=1, B=0, Alpha=$System.FN_Timer.Value))
    Spawn Time=0.1
    Use Spawn Probability=false
   ) {module=SpawnBurst_Instantaneous001}
  ]
  particleSpawnScript [
   /SetVariables(Particles.MeshIndex=&SelectIntFromArray(
     Direct Array Index=$Emitter.LoopCount
     Direct Array Mode=::NewEnumerator1
     Int Selection Array=DataInterface() {propertyValues {IntData [0,1]}}
    )) {isSetParameters=true}
   &InitializeParticle(
    Lifetime Max=0.3
    Lifetime Min=0.2
    Lifetime Mode=::NewEnumerator1
    Mesh Scale Max=<0.2 0.15 0.15>
    Mesh Scale Min=<0.15 0.1 0.1>
    Mesh Scale Mode=::NewEnumerator2
    UsePositionOffset=false
    Write Material Random=true
   )
   &CameraOffset(Camera Offset Amount=-50)
   &InitialMeshOrientation(
    Mesh Orientation Mode=::NewEnumerator2
    Orientation Vector=<0 0 1>
    Rotation=&RandomRangeVector(Minimum=<0 0 0>, Maximum=<0 0 1>)
    Use Rotation Vector=true
   ) {module=InitialMeshOrientation001}
  ]
  particleUpdateScript [
   &ParticleState()
   &DynamicMaterialParameters(
    Index 0 Param 1=&FloatFromCurve(FloatCurve=DataInterface() {propertyValues {Curve {keys [<at 0 1 RCIM_Cubic>, {arriveTangent=-1, interpMode=RCIM_Cubic, leaveTangent=-1, time=0.3, value=0.5}, <at 1 0>]}}})
    Index 0 Param 2=&FloatFromCurve(
     FloatCurve=DataInterface() {propertyValues {Curve {keys [
         {interpMode=RCIM_Cubic, leaveTangent=-0.9598, tangentMode=RCTM_Break, time=0, value=0.5}
         {arriveTangent=5, interpMode=RCIM_Cubic, leaveTangent=5, tangentMode=RCTM_Break, time=0.25, value=0.7}
         <at 0.3 1>
         {arriveTangent=-5.439, interpMode=RCIM_Cubic, leaveTangent=-5.439, tangentMode=RCTM_User, time=0.35, value=0.6}
         {arriveTangent=-0.3966, interpMode=RCIM_Cubic, tangentMode=RCTM_Break, time=1, value=0.05}
        ]}}}
     Scale Curve=&RandomRangeFloat(Minimum=0.1, Maximum=0.8)
    )
    Index 0 Param 3=&FloatFromCurve(FloatCurve=DataInterface() {propertyValues {Curve {keys [
         <at 0.1 0.5 RCIM_Cubic>
         <at 0.296 1>
         {arriveTangent=-2.9528, interpMode=RCIM_Cubic, leaveTangent=-2.9528, time=0.35, value=0.4}
         {arriveTangent=4.0435, interpMode=RCIM_Cubic, leaveTangent=4.0435, time=0.55, value=0.25}
         <at 0.993 3>
        ]}}})
    Index 0 Param 4=1
    Module.Param0WriteEnabled=true
    Module.Param1WriteEnabled=true
    Module.Param2WriteEnabled=true
    Module.Param3WriteEnabled=true
   )
   &ScaleColor(
    Scale Alpha=&FloatFromCurve(FloatCurve=DataInterface() {propertyValues {Curve {keys [<at 0.15 1 RCIM_Cubic>, {arriveTangent=-1.1765, interpMode=RCIM_Cubic, leaveTangent=-1.1765, time=0.796, value=0.245}, <at 1 0>]}}})
    Scale RGB=&VectorFromFloat(Value=&FloatFromCurve(FloatCurve=DataInterface() {propertyValues {Curve {keys [<at 0 1>,<at 0.1242 0.9958 RCIM_Constant>,<at 0.1835 0 RCIM_Constant>,<at 0.2506 0.9839 RCIM_Constant>,<at 0.5317 0 RCIM_Constant>,<at 0.6151 1 RCIM_Constant>,<at 1 0>]}}}))
    ScaleA=true
    ScaleRGB=true
   ) {module=ScaleColor002}
  ]
  rendererClasses [&NiagaraMeshRendererProperties]
  renderers [{rendererClass=&NiagaraMeshRendererProperties, rendererIndex=0}]
 }
 Lightning_Detail_FORWARD {
  emitterUpdateScript [
   &EmitterState(
    Inactive Response=::NewEnumerator1
    Life Cycle Mode=::NewEnumerator1
    Loop Delay=&UniformRangedFloat(Minimum=0.5, Maximum=0.8)
    Loop Duration=&UniformRangedFloat(Minimum=0.5, Maximum=0.8)
    Recalculate Duration Each Loop=true
    UseLoopDelay=false
   )
   &SpawnRate(
    Spawn Probability=0.5
    SpawnRate=&Lerp_Float(A=0, B=1, Alpha=$System.FN_Integrator001.Value)
    Use Spawn Probability=true
   )
   &SpawnBurst_Instantaneous(
    Spawn Count=&MakeIntFromFloat(Float=&Lerp_Float(A=0, B=1, Alpha=$System.FN_Integrator001.Value))
    Use Spawn Probability=false
   )
  ]
  particleSpawnScript [
   /SetVariables(Particles.MeshIndex=&RandomRangeInt(Minimum=0, Maximum=1)) {isSetParameters=true}
   &InitializeParticle(
    Lifetime Max=0.333
    Lifetime Min=0.222
    Lifetime Mode=::NewEnumerator1
    Mesh Scale Max=<0.4 0.4 0.25>
    Mesh Scale Min=<0.35 0.35 0.25>
    Mesh Scale Mode=::NewEnumerator2
    UsePositionOffset=false
    Write Material Random=true
   ) {module=InitializeParticle001}
   &DynamicMaterialParameters(
    Index 0 Param 2=&UniformRangedFloat(Minimum=0.2, Maximum=0.3)
    Module.Param0WriteEnabled=false
    Module.Param1WriteEnabled=true
    Module.Param2WriteEnabled=false
    Module.Param3WriteEnabled=false
   ) {module=DynamicMaterialParameters001}
   &InitialMeshOrientation(
    Mesh Orientation Mode=::NewEnumerator2
    Rotation=&SelectVectorFromArray(
     Array Sampling Mode=::NewEnumerator0
     Direct Array Index=?
     Vector Selection Array=DataInterface() {propertyValues {FloatData [<-0.05 -0.1 0.1>, <-0.05 -0.1 -0.1>, <-0.05 0.1 0.1>]}}
     Random Seed=0
    )
    Rotation Coordinate Space=::NewEnumerator2
    Use Rotation Vector=true
   )
   &CameraOffset(Camera Offset Amount=50)
   &OrientMeshToVector(
    Look At Direction:Vector3f=&NormalizeVector(Vector To Normalize=$Engine.Owner.Velocity, Safe Normalize=true, Safe Fallback Threshold=0, Safe Fallback Vector=<0 0 0.001>)
    Override Quat=false
    Reference Axis Vector:Vector3f=<-1 0 0>
    Side Axis Direction:Vector3f=$Engine.Owner.SystemXAxis
    Up Axis Direction:Vector3f=$Engine.Owner.SystemZAxis
    UseOptionalSideAxis=true
    UseOptionalUpAxis=true
   ) {module=OrientMeshToVector001}
  ]
  particleUpdateScript [
   &SkeletalMeshLocation(
    Inherit Velocity=false
    Interpolate Mesh Sampling=false
    MaskSpawnGroup=false
    Mesh Sampling Type=::NewEnumerator5
    Sampled Orientation Coordinate Space=::World
    Sampled Position Offset=&SelectVectorFromArray(
     Array Sampling Mode=::NewEnumerator0
     Direct Array Index=?
     Vector Selection Array=DataInterface() {propertyValues {FloatData [<-15 0 0>, <-5 0 0>, <5 0 0>, <15 0 0>]}}
     Random Seed=0
    )
    Skeletal Mesh=$User.Weapon
    Socket Sampling Mode=::NewEnumerator5
   )
   &ParticleState()
   &SolveForcesAndVelocity(Clamp Velocity=false, Limit Acceleration=false)
   &ScaleColor(
    Scale Alpha=&FloatFromCurve(FloatCurve=DataInterface() {propertyValues {Curve {keys [<at 0.15 1 RCIM_Cubic>, {arriveTangent=-1.1765, interpMode=RCIM_Cubic, leaveTangent=-1.1765, time=0.796, value=0.245}, <at 1 0>]}}})
    Scale RGB=&VectorFromFloat(Value=&FloatFromCurve(FloatCurve=DataInterface() {propertyValues {Curve {keys [<at 0 1>,<at 0.1242 0.9958 RCIM_Constant>,<at 0.1835 0 RCIM_Constant>,<at 0.284 0.984 RCIM_Constant>,<at 0.5317 0 RCIM_Constant>,<at 0.6402 0.984 RCIM_Constant>,<at 1 0>]}}}))
    ScaleA=true
    ScaleRGB=true
   ) {module=ScaleColor003}
   &DynamicMaterialParameters(
    Index 0 Param 1=&FloatFromCurve(FloatCurve=DataInterface() {propertyValues {Curve {keys [<at 0 1>,<at 0.7 0.35>,<at 1 0>]}}})
    Index 0 Param 2=0.05
    Index 0 Param 3=&FloatFromCurve(FloatCurve=DataInterface() {propertyValues {Curve {keys [<at 0.6 1>,<at 1 1>]}}}, Scale Curve=0.5)
    Index 0 Param 4=1
    Module.Param0WriteEnabled=true
    Module.Param1WriteEnabled=false
    Module.Param2WriteEnabled=true
    Module.Param3WriteEnabled=true
   )
  ]
  rendererClasses [&NiagaraMeshRendererProperties]
  renderers [{rendererClass=&NiagaraMeshRendererProperties, rendererIndex=0}]
 }
 Lightning_Main_FORWARD {
  emitterUpdateScript [
   &EmitterState(
    Inactive Response=::NewEnumerator2
    Life Cycle Mode=::NewEnumerator1
    Loop Delay=0.5
    Loop Duration=&RandomRangeFloat(Minimum=0.0888, Maximum=0.111)
    UseLoopDelay=false
   )
   /SetVariables(
    Emitter.M_Noise_01:NiagaraFloat=&RandomRangeFloat(Minimum=40, Maximum=50)
    Emitter.M_Noise_02:NiagaraFloat=&RandomRangeFloat(Minimum=1, Maximum=2)
    Emitter.M_Noise_03:NiagaraFloat=&RandomRangeFloat(Minimum=0, Maximum=1)
   ) {isSetParameters=true}
   &SpawnBurst_Instantaneous(
    Spawn Count=&MakeIntFromFloat(Float=&Lerp_Float(A=0, B=1, Alpha=$System.FN_Integrator001.Value))
    Use Spawn Probability=false
    UseLoopCountLimit=false
   )
  ]
  particleSpawnScript [
   /SetVariables(
    Particles.MeshIndex=&SelectIntFromArray(Direct Array Index=$Emitter.LoopCount, Direct Array Mode=::NewEnumerator1, Int Selection Array=$User.LightningOrder)
    Particles.bIsTravelingBolt:NiagaraBool=&RandomBool()
   ) {isSetParameters=true, module=SetVariables_#4}
   /SetVariables(Particles.MeshIndex=5) {enabled=false, isSetParameters=true}
   &InitializeParticle(
    Color=<rgba 0 48.0999 222 1>
    Lifetime=&MakeCustomFloatFromBool(
     Bool=$Particles.bIsTravelingBolt
     True Float=&RandomRangeFloat(Minimum=0.333, Maximum=0.111)
     False Float=0.222
    )
    Material Random=HlslExpression() {hlslExpression="(1.0) "}
    Mesh Scale=&SelectVectorFromArray(
     Direct Array Index=$Particles.MeshIndex
     Vector Selection Array=DataInterface() {propertyValues {FloatData [<2 2 4>, <1.5 2 3>, <1 1 3>, <1 -1 3>, <0.2 0.2 2.5>, <1 -1 3>]}}
    )
    Mesh Scale Mode=::NewEnumerator4
    Position Mode=::NewEnumerator0
    Write Material Random=true
   )
   &DynamicMaterialParameters(
    Index 0 Param 1=5
    Index 0 Param 2=&MakeCustomFloatFromBool(Bool=$Particles.bIsTravelingBolt, True Float=1, False Float=3)
    Index 0 Param 4=0.5
    Module.Param0WriteEnabled=true
    Module.Param1WriteEnabled=true
    Module.Param2WriteEnabled=true
    Module.Param3WriteEnabled=true
   )
   &CameraOffset(Camera Offset Amount=50)
   &InitialMeshOrientation(
    Mesh Axis To Orient=<0 0 1>
    Mesh Orientation Mode=::NewEnumerator2
    Orientation Vector=<0 0 -1>
    Rotation=&SelectVectorFromArray(
     Direct Array Index=$Particles.MeshIndex
     Vector Selection Array=DataInterface() {propertyValues {FloatData [<0.035 -0.1 0.1>, <0.035 -0.05 -0.1>, <-0.05 -0.05 0.1>, <-0.035 -0.025 0.1>, <0.025 -0.05 0.1>, <0.025 -0.05 -0.1>]}}
    )
    Rotation Coordinate Space=::NewEnumerator0
    Use Rotation Vector=true
   ) {module=InitialMeshOrientation001}
   &OrientMeshToVector(
    Look At Direction:Vector3f=&NormalizeVector(
     Vector To Normalize=&TransformVector(Vector=$Engine.Owner.Velocity, Source Space=::World, Destination Space=::Local)
     Safe Normalize=true
     Safe Fallback Threshold=0
     Safe Fallback Vector=<0 0 0.001>
    )
    Override Quat=false
    Reference Axis Vector:Vector3f=<-1 0 0>
    Side Axis Direction:Vector3f=$Engine.Owner.SystemXAxis
    Up Axis Direction:Vector3f=$Engine.Owner.SystemZAxis
    UseOptionalSideAxis=true
    UseOptionalUpAxis=true
   )
  ]
  particleUpdateScript [
   &ParticleState()
   &SkeletalMeshLocation(
    MaskSpawnGroup=false
    Mesh Sampling Type=::NewEnumerator5
    Orientation Sampling=::NewEnumerator1
    Sampled Position Offset=&SelectVectorFromArray(
     Array Sampling Mode=::NewEnumerator0
     Direct Array Index=$Emitter.LoopCount
     Vector Selection Array=DataInterface() {propertyValues {FloatData [<-15 0 -2>, <-5 5 -1>, <0 0 0>, <5 15 1>, <15 10 2>]}}
     Random Seed=0
    )
    Skeletal Mesh=$User.Weapon
    Socket Sampling Mode=::NewEnumerator5
   )
   &DynamicMaterialParameters(
    Index 0 Param 1=&MakeCustomFloatFromBool(
     Bool=$Particles.bIsTravelingBolt
     True Float=&RandomRangeFloat(Minimum=4, Maximum=5)
     False Float=&FloatFromCurve(
      FloatCurve=DataInterface() {propertyValues {Curve {keys [
          {tangentMode=RCTM_User, time=0, value=1}
          {arriveTangent=-0.0405, interpMode=RCIM_Cubic, leaveTangent=-0.0405, tangentMode=RCTM_User, time=0.5, value=1}
          {arriveTangent=-2.2998, interpMode=RCIM_Cubic, leaveTangent=-2.2998, tangentMode=RCTM_User, time=1, value=0.5}
         ]}}}
      Scale Curve=&RandomRangeFloat(Minimum=4, Maximum=5)
     )
    )
    Index 0 Param 3=&RandomRangeFloat(
     Minimum=&MakeCustomFloatFromBool(
      Bool=$Particles.bIsTravelingBolt
      True Float=&FloatFromCurve(FloatCurve=DataInterface() {propertyValues {Curve {keys [{arriveTangent=2.1548, interpMode=RCIM_Cubic, leaveTangent=2.1548, tangentMode=RCTM_User, time=0, value=0.4}, <at 1 1 RCIM_Cubic>]}}})
      False Float=&FloatFromCurve(FloatCurve=DataInterface() {propertyValues {Curve {keys [{arriveTangent=2.1548, interpMode=RCIM_Cubic, leaveTangent=2.1548, tangentMode=RCTM_User, time=0, value=0.4}, <at 1 1 RCIM_Cubic>]}}})
     )
     Maximum=&MakeCustomFloatFromBool(
      Bool=$Particles.bIsTravelingBolt
      True Float=$Particles.NormalizedAge
      False Float=&RandomRangeFloat(
       Minimum=&FloatFromCurve(FloatCurve=DataInterface() {propertyValues {Curve {keys [{arriveTangent=2.1548, interpMode=RCIM_Cubic, leaveTangent=2.1548, tangentMode=RCTM_User, time=0, value=0.4}, <at 1 1 RCIM_Cubic>]}}})
       Maximum=&FloatFromCurve(FloatCurve=DataInterface() {propertyValues {Curve {keys [{arriveTangent=2.1548, interpMode=RCIM_Cubic, leaveTangent=2.1548, tangentMode=RCTM_User, time=0, value=0.4}, <at 1 1 RCIM_Cubic>]}}})
      )
     )
    )
    Index 0 Param 4=&MakeCustomFloatFromBool(
     Bool=$Particles.bIsTravelingBolt
     True Float=1
     False Float=&FloatFromCurve(
      FloatCurve=DataInterface() {propertyValues {Curve {keys [{tangentMode=RCTM_User, time=0, value=0.7}, {tangentMode=RCTM_User, time=1, value=0.666}]}}}
      Scale Curve=&RandomRangeFloat(Minimum=0.5, Maximum=1)
     )
    )
    Module.Param0WriteEnabled=true
    Module.Param1WriteEnabled=false
    Module.Param2WriteEnabled=true
    Module.Param3WriteEnabled=true
   ) {module=DynamicMaterialParameters001}
   &ScaleColor(
    Scale RGB=&Multiply_VectorByFloat(Vector=<1 1 1>, Float=$System.TimeOfDayBrightnessMultiplier)
    ScaleA=true
    ScaleRGB=true
   ) {module=ScaleColor001}
  ]
  rendererClasses [&NiagaraMeshRendererProperties]
  renderers [{rendererClass=&NiagaraMeshRendererProperties, rendererIndex=0}]
 }
 Lightning_Main_IDLE {
  emitterUpdateScript [
   &EmitterState(
    Inactive Response=::NewEnumerator2
    Life Cycle Mode=::NewEnumerator1
    Loop Delay=0.5
    Loop Duration=&RandomRangeFloat(Minimum=0.111, Maximum=0.222)
    UseLoopDelay=false
   )
   /SetVariables(Emitter.M_Noise_01:NiagaraFloat=&RandomRangeFloat(Minimum=15, Maximum=25)) {isSetParameters=true}
   &SpawnBurst_Instantaneous(
    Spawn Count=&MakeIntFromFloat(Float=&Lerp_Float(A=1, B=0, Alpha=$System.FN_Timer.Value))
    Use Spawn Probability=false
    UseLoopCountLimit=false
   )
   &SpawnBurst_Instantaneous(
    Spawn Count=&MakeIntFromFloat(Float=&Lerp_Float(A=1, B=0, Alpha=$System.FN_Timer.Value))
    Spawn Time=&RandomRangeFloat(Minimum=0.1, Maximum=0.5)
    Use Spawn Probability=false
    UseLoopCountLimit=false
   ) {module=SpawnBurst_Instantaneous001}
  ]
  particleSpawnScript [
   /SetVariables(
    Particles.MeshIndex=&SelectIntFromArray(Direct Array Index=$Emitter.LoopCount, Direct Array Mode=::NewEnumerator1, Int Selection Array=$User.LightningOrder)
    Particles.bIsTravelingBolt:NiagaraBool=&RandomBool()
   ) {isSetParameters=true}
   &InitializeParticle(
    Color=<rgba 1 50 225 1>
    Lifetime Max=0.444
    Lifetime Min=0.222
    Lifetime Mode=::NewEnumerator1
    Material Random=HlslExpression() {hlslExpression=(1.0)}
    Mesh Scale Max=<3 3 3>
    Mesh Scale Min=<1 1 0.5>
    Mesh Scale Mode=::NewEnumerator2
    Position Mode=::NewEnumerator0
    Write Material Random=true
   )
   &DynamicMaterialParameters(
    Index 0 Param 1=3
    Index 0 Param 2=&MakeCustomFloatFromBool(
     Bool=$Particles.bIsTravelingBolt
     True Float=&RandomRangeFloat(Minimum=0.5, Maximum=2)
     False Float=&RandomRangeFloat(Minimum=1, Maximum=2)
    )
    Index 0 Param 4=0.5
    Index 1 Param 1=1
    Module.Param0WriteEnabled=true
    Module.Param0WriteEnabled1=true
    Module.Param1WriteEnabled=true
    Module.Param1WriteEnabled1=true
    Module.Param2WriteEnabled=true
    Module.Param2WriteEnabled1=true
    Module.Param3WriteEnabled=true
    Module.Param3WriteEnabled1=true
    Write Parameter Index 1=true
   )
   &CameraOffset(Camera Offset Amount=-50)
   &InitialMeshOrientation(
    Mesh Orientation Mode=::NewEnumerator2
    Orientation Vector=<0 0 1>
    Rotation=&UniformRangedVector(Minimum=<-1 -1 -1>, Maximum=<1 1 1>)
    Use Rotation Vector=true
   ) {module=InitialMeshOrientation001}
  ]
  particleUpdateScript [
   &ParticleState()
   &DynamicMaterialParameters(
    Index 0 Param 1=&MakeCustomFloatFromBool(
     Bool=$Particles.bIsTravelingBolt
     True Float=5
     False Float=&FloatFromCurve(
      FloatCurve=DataInterface() {propertyValues {Curve {keys [
          {tangentMode=RCTM_User, time=0, value=1}
          {arriveTangent=-0.0405, interpMode=RCIM_Cubic, leaveTangent=-0.0405, tangentMode=RCTM_User, time=0.5, value=1}
          {arriveTangent=-8.0089, interpMode=RCIM_Cubic, leaveTangent=-8.0089, tangentMode=RCTM_User, time=1, value=0}
         ]}}}
      Scale Curve=&RandomRangeFloat(Minimum=4, Maximum=5)
     )
    )
    Index 0 Param 3=&MakeCustomFloatFromBool(
     Bool=$Particles.bIsTravelingBolt
     True Float=$Particles.NormalizedAge
     False Float=&FloatFromCurve(
      FloatCurve=DataInterface() {propertyValues {Curve {keys [{tangentMode=RCTM_User, time=0, value=0}, {tangentMode=RCTM_User, time=1, value=1}]}}}
      Scale Curve=&RandomRangeFloat(Minimum=0.5, Maximum=1)
     )
    )
    Index 0 Param 4=&MakeCustomFloatFromBool(
     Bool=$Particles.bIsTravelingBolt
     True Float=&RandomRangeFloat(Minimum=0.75, Maximum=2)
     False Float=&FloatFromCurve(
      FloatCurve=DataInterface() {propertyValues {Curve {keys [{tangentMode=RCTM_User, time=0, value=0.333}, {tangentMode=RCTM_User, time=1, value=0.666}]}}}
      Scale Curve=&RandomRangeFloat(Minimum=1, Maximum=2)
     )
    )
    Module.Param0WriteEnabled=true
    Module.Param1WriteEnabled=false
    Module.Param2WriteEnabled=true
    Module.Param3WriteEnabled=true
   ) {module=DynamicMaterialParameters001}
   &ScaleColor(
    Scale Alpha=&FloatFromCurve(FloatCurve=DataInterface() {propertyValues {Curve {keys [<at 0 1>,<at 0.1242 0.9958 RCIM_Constant>,<at 0.1835 0 RCIM_Constant>,<at 0.284 0.984 RCIM_Constant>,<at 0.5317 0 RCIM_Constant>,<at 0.6402 0.984 RCIM_Constant>,<at 1 0>]}}})
    ScaleA=true
    ScaleRGB=true
   ) {module=ScaleColor003}
  ]
  rendererClasses [&NiagaraMeshRendererProperties]
  renderers [{rendererClass=&NiagaraMeshRendererProperties, rendererIndex=0}]
 }
 Lightning_Main_Small_FORWARD {
  emitterUpdateScript [
   &EmitterState(
    Inactive Response=::NewEnumerator2
    Life Cycle Mode=::NewEnumerator1
    Loop Delay=0.5
    Loop Duration=&RandomRangeFloat(Minimum=0.222, Maximum=0.333)
    UseLoopDelay=false
   )
   /SetVariables(
    Emitter.M_Noise_01:NiagaraFloat=&RandomRangeFloat(Minimum=15, Maximum=25)
    Emitter.M_Noise_02:NiagaraFloat=&RandomRangeFloat(Minimum=5, Maximum=10)
    Emitter.M_Noise_03:NiagaraFloat=&RandomRangeFloat(Minimum=1, Maximum=2)
   ) {isSetParameters=true}
   &SpawnBurst_Instantaneous(
    Spawn Count=&MakeIntFromFloat(Float=&Lerp_Float(A=0, B=1, Alpha=$System.FN_Integrator001.Value))
    Spawn Time=&RandomRangeFloat(Minimum=0, Maximum=0.2)
    Use Spawn Probability=false
    UseLoopCountLimit=false
   )
  ]
  particleSpawnScript [
   /SetVariables(
    Particles.MeshIndex=&SelectIntFromArray(Direct Array Index=$Emitter.LoopCount, Direct Array Mode=::NewEnumerator1, Int Selection Array=$User.LightningOrder)
    Particles.bIsTravelingBolt:NiagaraBool=&RandomBool()
   ) {isSetParameters=true, module=SetVariables_#4}
   &InitializeParticle(
    Color=<rgba 0 48.0999 222 1>
    Lifetime=&MakeCustomFloatFromBool(Bool=$Particles.bIsTravelingBolt, True Float=0.333, False Float=0.222)
    Material Random=HlslExpression() {hlslExpression="(1.0) "}
    Mesh Scale=&SelectVectorFromArray(
     Direct Array Index=$Particles.MeshIndex
     Vector Selection Array=DataInterface() {propertyValues {FloatData [<2 2 4>, <1.5 2 4>, <1 1 4>, <1 1 4>, <1 1 4>, <1 1 4>]}}
    )
    Mesh Scale Mode=::NewEnumerator4
    Position Mode=::NewEnumerator0
    Write Material Random=true
   )
   &DynamicMaterialParameters(
    Index 0 Param 1=5
    Index 0 Param 2=&MakeCustomFloatFromBool(Bool=$Particles.bIsTravelingBolt, True Float=0.75, False Float=3)
    Index 0 Param 4=0.5
    Module.Param0WriteEnabled=true
    Module.Param1WriteEnabled=true
    Module.Param2WriteEnabled=true
    Module.Param3WriteEnabled=true
   )
   &CameraOffset(Camera Offset Amount=50)
   &InitialMeshOrientation(
    Mesh Axis To Orient=<0 0 1>
    Mesh Orientation Mode=::NewEnumerator2
    Orientation Vector=<0 0 -1>
    Rotation=&SelectVectorFromArray(
     Direct Array Index=$Particles.MeshIndex
     Vector Selection Array=DataInterface() {propertyValues {FloatData [<0.05 -0.1 0.1>, <0.05 -0.05 -0.1>, <-0.1 -0.05 0.1>, <-0.075 -0.025 0.1>, <0.025 -0.05 0.1>, <0.05 -0.05 -0.1>]}}
    )
    Rotation Coordinate Space=::NewEnumerator0
    Use Rotation Vector=true
   ) {module=InitialMeshOrientation001}
   &OrientMeshToVector(
    Look At Direction:Vector3f=&NormalizeVector(
     Vector To Normalize=&TransformVector(Vector=$Engine.Owner.Velocity, Source Space=::World, Destination Space=::Local)
     Safe Normalize=true
     Safe Fallback Threshold=0
     Safe Fallback Vector=<0 0 0.001>
    )
    Override Quat=false
    Reference Axis Vector:Vector3f=<-1 0 0>
    Side Axis Direction:Vector3f=$Engine.Owner.SystemXAxis
    Up Axis Direction:Vector3f=$Engine.Owner.SystemZAxis
    UseOptionalSideAxis=true
    UseOptionalUpAxis=true
   ) {module=OrientMeshToVector001}
  ]
  particleUpdateScript [
   &ParticleState()
   &SkeletalMeshLocation(
    MaskSpawnGroup=false
    Mesh Sampling Type=::NewEnumerator5
    Orientation Sampling=::NewEnumerator1
    Sampled Position Offset=&SelectVectorFromArray(
     Array Sampling Mode=::NewEnumerator0
     Direct Array Index=?
     Vector Selection Array=DataInterface() {propertyValues {FloatData [<-15 0 -2>, <-5 0 -1>, <0 0 0>, <5 0 1>, <15 0 2>]}}
     Random Seed=0
    )
    Skeletal Mesh=$User.Weapon
    Socket Sampling Mode=::NewEnumerator5
   )
   &DynamicMaterialParameters(
    Index 0 Param 1=&MakeCustomFloatFromBool(
     Bool=$Particles.bIsTravelingBolt
     True Float=5
     False Float=&FloatFromCurve(
      FloatCurve=DataInterface() {propertyValues {Curve {keys [
          {tangentMode=RCTM_User, time=0, value=1}
          {arriveTangent=-0.0405, interpMode=RCIM_Cubic, leaveTangent=-0.0405, tangentMode=RCTM_User, time=0.5, value=1}
          {arriveTangent=0.009, interpMode=RCIM_Cubic, leaveTangent=0.009, tangentMode=RCTM_User, time=1, value=0}
         ]}}}
      Scale Curve=5
     )
    )
    Index 0 Param 3=&MakeCustomFloatFromBool(
     Bool=$Particles.bIsTravelingBolt
     True Float=&FloatFromCurve(FloatCurve=DataInterface() {propertyValues {Curve {keys [{tangentMode=RCTM_User, time=0, value=0.35}, {tangentMode=RCTM_User, time=1, value=1}]}}})
     False Float=&FloatFromCurve(FloatCurve=DataInterface() {propertyValues {Curve {keys [{tangentMode=RCTM_User, time=0, value=0.7}, {tangentMode=RCTM_User, time=1, value=1}]}}})
    )
    Index 0 Param 4=&MakeCustomFloatFromBool(
     Bool=$Particles.bIsTravelingBolt
     True Float=&RandomRangeFloat(Minimum=0.5, Maximum=1)
     False Float=&FloatFromCurve(
      FloatCurve=DataInterface() {propertyValues {Curve {keys [{tangentMode=RCTM_User, time=0, value=0.333}, {tangentMode=RCTM_User, time=1, value=0.666}]}}}
      Scale Curve=&Lerp_Float(A=3, B=1, Alpha=$System.FN_Timer.Value)
     )
    )
    Module.Param0WriteEnabled=true
    Module.Param1WriteEnabled=false
    Module.Param2WriteEnabled=true
    Module.Param3WriteEnabled=true
   ) {module=DynamicMaterialParameters001}
   &ScaleColor(
    Scale RGB=&Multiply_VectorByFloat(Vector=<1 1 1>, Float=$System.TimeOfDayBrightnessMultiplier)
    ScaleA=true
    ScaleRGB=true
   ) {module=ScaleColor001}
  ]
  rendererClasses [&NiagaraMeshRendererProperties]
  renderers [{rendererClass=&NiagaraMeshRendererProperties, rendererIndex=0}]
 }
 Lightning_Meshes_IDLE {
  emitterUpdateScript [
   &EmitterState(
    Inactive Response=::NewEnumerator2
    Life Cycle Mode=::NewEnumerator1
    Loop Delay=0.5
    Loop Duration=&Lerp_Float(
     A=&RandomRangeFloat(Minimum=0.0666, Maximum=0.1666)
     B=&RandomRangeFloat(Minimum=0.05, Maximum=0.2)
     Alpha=$System.FN_Integrator001.Value
    )
    Recalculate Duration Each Loop=true
    UseLoopDelay=true
   )
   &SpawnBurst_Instantaneous(Use Spawn Probability=false, UseLoopCountLimit=false)
  ]
  particleSpawnScript [
   /SetVariables(
    Particles.MeshIndex=&SelectIntFromArray(Direct Array Index=$Emitter.LoopCount, Direct Array Mode=::NewEnumerator1, Int Selection Array=$User.LightningOrder)
    Particles.bIsTravelingBolt:NiagaraBool=&RandomBool()
   ) {isSetParameters=true, module=SetVariables_#4}
   &InitializeParticle(
    Color=<rgba 0 48.0999 222 1>
    Lifetime=&Lerp_Float(
     A=&MakeCustomFloatFromBool(Bool=$Particles.bIsTravelingBolt, True Float=0.444, False Float=0.222)
     B=&MakeCustomFloatFromBool(Bool=$Particles.bIsTravelingBolt, True Float=0.555, False Float=0.225)
     Alpha=$System.FN_Integrator001.Value
    )
    Position Offset=<0 0 -25>
    UsePositionOffset=true
    Write Material Random=true
   )
   &DynamicMaterialParameters(
    Index 0 Param 1=1.5
    Index 0 Param 2=&MakeCustomFloatFromBool(Bool=$Particles.bIsTravelingBolt, True Float=0.75, False Float=3)
    Index 0 Param 4=0.5
    Module.Param0WriteEnabled=true
    Module.Param1WriteEnabled=true
    Module.Param2WriteEnabled=true
    Module.Param3WriteEnabled=true
   )
   &InitialMeshOrientation(
    Mesh Orientation Mode=::NewEnumerator7
    Rotation=&SelectVectorFromArray(
     Direct Array Index=/Niagara/DynamicInputs/Divide/Divide_Int(A:NiagaraInt32=$Emitter.LoopCount, B:NiagaraInt32=4)
     Direct Array Mode=::NewEnumerator1
     Vector Selection Array=DataInterface() {propertyValues {FloatData [<0 0 0.5>, <0 0 0>]}}
    )
    Use Rotation Vector=true
   )
   &CameraOffset(Camera Offset Amount=5)
  ]
  particleUpdateScript [
   &ParticleState()
   &DynamicMaterialParameters(
    Index 0 Param 1=&MakeCustomFloatFromBool(
     Bool=$Particles.bIsTravelingBolt
     True Float=1.5
     False Float=&FloatFromCurve(
      FloatCurve=DataInterface() {propertyValues {Curve {keys [<at 0.666 1>,<at 1 0>]}}}
      Scale Curve=&Lerp_Float(A=1.5, B=5, Alpha=$System.FN_Integrator001.Value)
     )
    )
    Index 0 Param 3=&MakeCustomFloatFromBool(
     Bool=$Particles.bIsTravelingBolt
     True Float=$Particles.NormalizedAge
     False Float=&FloatFromCurve(
      FloatCurve=DataInterface() {propertyValues {Curve {keys [<at 0 0>,<at 0.333 0.5>]}}}
      Scale Curve=&Lerp_Float(A=1, B=0.5, Alpha=$System.FN_Integrator001.Value)
     )
    )
    Index 0 Param 4=&MakeCustomFloatFromBool(
     Bool=$Particles.bIsTravelingBolt
     True Float=0.5
     False Float=&FloatFromCurve(FloatCurve=DataInterface() {propertyValues {Curve {keys [{tangentMode=RCTM_User, time=0, value=0.333}, {tangentMode=RCTM_User, time=1, value=0.666}]}}})
    )
    Module.Param0WriteEnabled=true
    Module.Param1WriteEnabled=false
    Module.Param2WriteEnabled=true
    Module.Param3WriteEnabled=true
   ) {module=DynamicMaterialParameters001}
   &ScaleColor(
    Scale RGB=&VectorFromFloat(Value=&Lerp_Float(A=0.133, B=1, Alpha=$System.Flicker))
    ScaleA=false
    ScaleRGB=true
   )
  ]
  rendererClasses [&NiagaraMeshRendererProperties]
  renderers [{rendererClass=&NiagaraMeshRendererProperties, rendererIndex=0}]
 }
 Lightning_Residue_BURST {
  emitterUpdateScript [
   &EmitterState(Life Cycle Mode=::NewEnumerator1, Loop Behavior=::NewEnumerator1, Loop Duration=1, UseLoopDelay=false)
   &SpawnBurst_Instantaneous(Spawn Count=3, Spawn Time=0.18, Use Spawn Probability=false)
   &SpawnBurst_Instantaneous(Spawn Count=2, Spawn Time=0.3, Use Spawn Probability=false) {module=SpawnBurst_Instantaneous001}
  ]
  particleSpawnScript [
   &InitializeParticle(
    Color=<rgba 2.2611 16.188 20 5>
    Lifetime Max=0.1
    Lifetime Min=0.08
    Lifetime Mode=::NewEnumerator1
    Mesh Scale Max=<8 8 8>
    Mesh Scale Min=<3 3 8>
    Mesh Scale Mode=::NewEnumerator2
    UsePositionOffset=false
    Write Material Random=true
   ) {module=InitializeParticle001}
   &InitialMeshOrientation(
    Mesh Orientation Mode=::NewEnumerator2
    Rotation=&RandomRangeVector(Minimum=<0 0 0>, Maximum=<1 0 0>)
    Use Rotation Vector=true
   ) {module=InitialMeshOrientation002}
   &DynamicMaterialParameters(
    Index 0 Param 1=&UniformRangedFloat(Minimum=5, Maximum=10)
    Index 0 Param 3=&UniformRangedFloat(Minimum=0.2, Maximum=1)
    Index 0 Param 4=&UniformRangedFloat(Minimum=0, Maximum=1)
    Module.Param0WriteEnabled=true
    Module.Param1WriteEnabled=false
    Module.Param2WriteEnabled=true
    Module.Param3WriteEnabled=true
   )
   &ShapeLocation(MaskSpawnGroup=false, Override Seed=false, Sphere Radius=10, UseOwnerScale=true)
   &AddVelocity(Coordinate Space=::World, Velocity=<0 0 1>)
  ]
  particleUpdateScript [
   &SkeletalMeshLocation(
    Inherit Velocity=false
    Interpolate Mesh Sampling=false
    MaskSpawnGroup=false
    Mesh Sampling Type=::NewEnumerator5
    Sampled Orientation Coordinate Space=::World
    Sampled Position Offset=<0 0 -30>
    Skeletal Mesh=$User.Weapon
    Socket Sampling Mode=::NewEnumerator5
   )
   &ParticleState()
   &SolveForcesAndVelocity(Clamp Velocity=false, Limit Acceleration=false)
   &ScaleColor(
    Scale Alpha=&FloatFromCurve(
     FloatCurve=DataInterface() {propertyValues {Curve {keys [<at 0 1 RCIM_Constant>,<at 0.1189 0 RCIM_Constant>,<at 0.2516 0.997 RCIM_Constant>,<at 0.6735 -0.0069 RCIM_Constant>,<at 0.7649 1.0117 RCIM_Constant>,<at 1 0 RCIM_Constant>]}}}
     Scale Curve=&Multiply_Float(
      A=&FloatFromCurve(FloatCurve=DataInterface() {propertyValues {Curve {keys [<at 0 1 RCIM_Constant>,<at 0.1561 -0.0068 RCIM_Constant>,<at 0.3011 0.9877 RCIM_Constant>,<at 0.6424 0.0069 RCIM_Constant>,<at 0.7618 0.9877 RCIM_Constant>,<at 1 0 RCIM_Constant>]}}})
      B=&UniformRangedFloat(Minimum=1, Maximum=1)
     )
    )
    Scale RGB=&VectorFromFloat(Value=&UniformRangedFloat(Minimum=1, Maximum=5))
    ScaleA=true
    ScaleRGB=true
   ) {module=ScaleColor001}
   &DynamicMaterialParameters(
    Index 0 Param 2=&FloatFromCurve(FloatCurve=DataInterface() {propertyValues {Curve {keys [{interpMode=RCIM_Cubic, leaveTangent=-2.846, tangentMode=RCTM_Break, time=0, value=1}, <at 1 -0.25>]}}})
    Module.Param0WriteEnabled=false
    Module.Param1WriteEnabled=true
    Module.Param2WriteEnabled=false
    Module.Param3WriteEnabled=false
   ) {module=DynamicMaterialParameters001}
  ]
  rendererClasses [&NiagaraMeshRendererProperties]
  renderers [{rendererClass=&NiagaraMeshRendererProperties, rendererIndex=0}]
 }
 Lightning_Sprite_BURST {
  emitterUpdateScript [
   &EmitterState(Life Cycle Mode=::NewEnumerator1, Loop Behavior=::NewEnumerator1, Loop Delay=0.3, Loop Duration=0.5, UseLoopDelay=true)
   &SpawnRate(SpawnRate=10, Use Spawn Probability=false)
  ]
  particleSpawnScript [
   &InitializeParticle(
    Lifetime Max=0.25
    Lifetime Min=0.15
    Lifetime Mode=::NewEnumerator1
    Sprite Rotation Mode=::NewEnumerator2
    Sprite Size Mode=::NewEnumerator1
    Uniform Sprite Size Max=70
    Uniform Sprite Size Min=50
    UsePositionOffset=false
    Write Material Random=true
   ) {module=InitializeParticle001}
   /SetVariables(Particles.SubImageIndex=&UniformRangedFloat(Minimum=0, Maximum=3)) {isSetParameters=true, module=SetVariables_#3}
   &CameraOffset()
   &ShapeLocation(MaskSpawnGroup=false, Override Seed=false, Sphere Radius=15, UseOwnerScale=true)
  ]
  particleUpdateScript [
   &DynamicMaterialParameters(
    Index 0 Param 1=&RandomRangeFloat(Minimum=0, Maximum=1)
    Index 0 Param 3=&FloatFromCurve(FloatCurve=DataInterface() {propertyValues {Curve {keys [
         {arriveTangent=4.5017, interpMode=RCIM_Cubic, leaveTangent=4.5017, tangentMode=RCTM_User, time=0, value=0}
         {interpMode=RCIM_Cubic, tangentMode=RCTM_User, time=0.3, value=1}
         {arriveTangent=-1.4286, interpMode=RCIM_Cubic, leaveTangent=-1.4286, time=0.7, value=1}
         {tangentMode=RCTM_User, time=1, value=0}
        ]}}})
    Module.Param0WriteEnabled=true
    Module.Param1WriteEnabled=true
    Module.Param2WriteEnabled=true
    Module.Param3WriteEnabled=false
   )
   &ParticleState()
   &SolveForcesAndVelocity(Clamp Velocity=false, Limit Acceleration=false)
   &ScaleColor(
    Scale Alpha=&FloatFromCurve(FloatCurve=DataInterface() {propertyValues {Curve {keys [<at 0.75 1 RCIM_Cubic>, {arriveTangent=-10.3917, interpMode=RCIM_Cubic, leaveTangent=-10.3917, tangentMode=RCTM_User, time=1, value=0}]}}})
    Scale RGB=&Multiply_VectorByFloat(
     Vector=&MakeVectorFromLinearColorRGB(Color=&ColorFromCurve(ColorCurve=DataInterface() {propertyValues {
         AlphaCurve {keys []}
         BlueCurve {keys [<at 0 5>,<at 0.2001 5>,<at 0.7465 2>,<at 1 1 RCIM_Cubic>]}
         GreenCurve {keys [<at 0 1.9663>,<at 0.2001 2.4757>,<at 0.7465 0.3455>,<at 1 0.1257 RCIM_Cubic>]}
         RedCurve {keys [<at 0 0>,<at 0.2001 0>,<at 0.7465 0>,<at 1 0 RCIM_Cubic>]}
        }}))
     Float=1
    )
    ScaleA=true
    ScaleRGB=true
   ) {module=ScaleColor002}
   &FN_Light_v2(
    Light Exponent=3
    Light Radius=50
    LightEnabled=true
    Scale Color:NiagaraFloat=&FloatFromCurve(FloatCurve=DataInterface() {propertyValues {Curve {keys [
         {interpMode=RCIM_Cubic, tangentMode=RCTM_User, time=0.3, value=0}
         {interpMode=RCIM_Cubic, tangentMode=RCTM_User, time=0.5, value=1}
         {tangentMode=RCTM_User, time=0.7, value=0}
        ]}}})
   )
  ]
  rendererClasses [&NiagaraSpriteRendererProperties, &NiagaraLightRendererProperties]
  renderers [
   {rendererClass=&NiagaraSpriteRendererProperties}
   {rendererClass=&NiagaraLightRendererProperties, rendererIndex=1}
  ]
 }
 Lightning_Sprite_IDLE {
  emitterUpdateScript [
   &EmitterState(Life Cycle Mode=::NewEnumerator1, Loop Behavior=::NewEnumerator1, Loop Delay=0.5, Loop Duration Mode=::NewEnumerator1, UseLoopDelay=true)
   &SpawnRate(Spawn Probability=0.5, SpawnRate=16, Use Spawn Probability=true)
  ]
  particleSpawnScript [
   &InitializeParticle(
    Color=<rgba 0 1.3 6 1>
    Lifetime Max=0.166
    Lifetime Min=0.133
    Lifetime Mode=::NewEnumerator1
    Sprite Rotation Mode=::NewEnumerator2
    Sprite Size Mode=::NewEnumerator1
    Sprite UV Mode=::NewEnumerator3
    Uniform Sprite Size Max=29
    Uniform Sprite Size Min=21
    UsePositionOffset=true
    Write Material Random=true
   ) {module=InitializeParticle001}
   /SetVariables(Particles.SubImageIndex=&UniformRangedFloat(Minimum=0, Maximum=3)) {isSetParameters=true, module=SetVariables_#3}
   &CameraOffset(Camera Offset Amount=9)
   &ShapeLocation(Box Size=<22 11 11>, MaskSpawnGroup=false, Override Seed=false, Shape Primitive=::NewEnumerator2, Surface Only=false, UseOwnerScale=true)
  ]
  particleUpdateScript [
   &DynamicMaterialParameters(
    Index 0 Param 1=&RandomRangeFloat(Minimum=0, Maximum=1)
    Index 0 Param 2=144
    Index 0 Param 3=&FloatFromCurve(FloatCurve=DataInterface() {propertyValues {Curve {keys [
         {arriveTangent=-19.4071, interpMode=RCIM_Cubic, leaveTangent=-19.4071, tangentMode=RCTM_User, time=0, value=1}
         {interpMode=RCIM_Cubic, tangentMode=RCTM_Break, time=0.1, value=0}
         {arriveTangent=3.3801, interpMode=RCIM_Cubic, leaveTangent=3.3801, tangentMode=RCTM_User, time=1, value=1}
        ]}}})
    Module.Param0WriteEnabled=true
    Module.Param1WriteEnabled=true
    Module.Param2WriteEnabled=true
    Module.Param3WriteEnabled=false
   )
   &ParticleState()
   &ScaleColor(
    Scale Alpha=&FloatFromCurve(
     FloatCurve=DataInterface() {propertyValues {Curve {keys [<at 0.75 1 RCIM_Cubic>, {arriveTangent=-10.3917, interpMode=RCIM_Cubic, leaveTangent=-10.3917, tangentMode=RCTM_User, time=1, value=0}]}}}
     Scale Curve=&Lerp_Float(A=0.1, B=1, Alpha=$System.Flicker)
    )
    ScaleA=true
    ScaleRGB=false
   ) {module=ScaleColor002}
   &ScaleSpriteSize(
    Uniform Curve Scale=&Lerp_Float(A=0.9, B=1, Alpha=$System.Flicker)
    Uniform Curve Sprite Scale=DataInterface() {propertyValues {Curve {keys [{arriveTangent=1.4996, time=0, value=0.9}, <at 0.2 1>, <at 0.7 0.98>, {leaveTangent=-0.2857, time=1, value=1}]}}}
   )
  ]
  rendererClasses [&NiagaraSpriteRendererProperties]
  renderers [{rendererClass=&NiagaraSpriteRendererProperties}]
 }
 Lightning_Strike_BURST {
  emitterUpdateScript [
   &EmitterState(Life Cycle Mode=::NewEnumerator1, Loop Behavior=::NewEnumerator1, Loop Duration=1, UseLoopDelay=false)
   &SpawnBurst_Instantaneous(Spawn Count=2, Use Spawn Probability=false)
   &SpawnBurst_Instantaneous(Spawn Count=1, Spawn Time=0.1, Use Spawn Probability=false) {module=SpawnBurst_Instantaneous001}
  ]
  particleSpawnScript [
   &InitializeParticle(
    Color=&MakeLinearColorFromVectorAndFloat("Vector (RGB)"=<1 1 1>, "Float (Alpha)"=5)
    Lifetime Max=0.15
    Lifetime Min=0.1
    Lifetime Mode=::NewEnumerator1
    Mesh Scale=<0.17 0.17 0.18>
    Mesh Scale Mode=::NewEnumerator4
    UsePositionOffset=false
    Write Material Random=true
   ) {module=InitializeParticle001}
   &InitialMeshOrientation(
    Mesh Orientation Mode=::NewEnumerator2
    Orientation Vector=<0 0 -1>
    Rotation=&RandomRangeVector(Minimum=<0 0 0>, Maximum=<0 0 1>)
    Use Rotation Vector=true
   ) {module=InitialMeshOrientation002}
   &AddVelocity(Coordinate Space=::World, Velocity=<0 0 1>)
   &SkeletalMeshLocation(
    Inherit Velocity=false
    Interpolate Mesh Sampling=false
    MaskSpawnGroup=false
    Mesh Sampling Type=::NewEnumerator5
    Module.Override Randomness=false
    Sampled Orientation Coordinate Space=::World
    Sampled Position Offset=<0 0 30>
    Skeletal Mesh=$User.Weapon
   ) {module=SkeletalMeshLocation001}
  ]
  particleUpdateScript [
   &DynamicMaterialParameters(
    Index 0 Param 1=&UniformRangedFloat(Minimum=1, Maximum=5)
    Index 0 Param 2=&UniformRangedFloat(Minimum=0, Maximum=1)
    Index 0 Param 3=1
    Index 0 Param 4=1
    Module.Param0WriteEnabled=true
    Module.Param1WriteEnabled=true
    Module.Param2WriteEnabled=true
    Module.Param3WriteEnabled=true
   )
   &ParticleState()
   &SolveForcesAndVelocity(Clamp Velocity=false, Limit Acceleration=false)
  ]
  rendererClasses [&NiagaraMeshRendererProperties]
  renderers [{rendererClass=&NiagaraMeshRendererProperties, rendererIndex=0}]
 }
 Sparks_BURST {
  emitterUpdateScript [
   &EmitterState(
    Enable Distance Culling=true
    Life Cycle Mode=::NewEnumerator1
    Loop Behavior=::NewEnumerator1
    Loop Delay=0.1
    Loop Duration=1
    Max Distance Response=::KillImmediately
    Module.bApplySpawnCountScaleByDistance=true
    Module.bUseMaxDistance=true
    Module.bUseMinDistance=false
    Scalability Mode=::NewEnumerator1
    Scale Spawn Count=true
    Spawn Count Scale By Distance=DataInterface() {propertyValues {Curve {keys [<at 0.3676 0.9657>,<at 0.9962 0.4982>]}}}
    UseLoopDelay=true
   )
   &SpawnBurst_Instantaneous(Spawn Count=&UniformRangedInt(Minimum=12, Maximum=15), Use Spawn Probability=false)
   &SpawnRate(
    Spawn Probability=&FloatFromCurve(
     FloatCurve=DataInterface() {propertyValues {Curve {keys [{arriveTangent=-12.2872, interpMode=RCIM_Cubic, leaveTangent=-12.2872, tangentMode=RCTM_User, time=0, value=1}, <at 0.1 0>]}}}
     CurveIndex=$System.Age
    )
    SpawnRate=444
    Use Spawn Probability=true
   )
  ]
  particleSpawnScript [
   &InitializeParticle(
    Lifetime Max=0.45
    Lifetime Min=0.166
    Lifetime Mode=::NewEnumerator1
    Sprite Size Mode=::NewEnumerator1
    Uniform Sprite Size Max=8
    Uniform Sprite Size Min=2.5
    UsePositionOffset=false
    Write Material Random=true
   )
   &SphereLocation(
    MaskSpawnGroup=false
    Override Seed=false
    Sphere Radius=5
    Surface Only=false
    UseNonUniformScale=false
    UseOffset=false
    UseSphereOrientation=false
   ) {module=SphereLocation001}
   &AddVelocityFromPoint(Module.Use Falloff Distance=false, Velocity Strength=&UniformRangedFloat(Minimum=350, Maximum=750))
  ]
  particleUpdateScript [
   &ParticleState()
   &GravityForce(Gravity=<0 0 -666>)
   &Drag(Drag=&UniformRangedFloat(Minimum=2, Maximum=6), Use Linear Drag=true, Use Rotational Drag=false)
   &SolveForcesAndVelocity(Clamp Velocity=false, Limit Acceleration=false)
   &ScaleSpriteSize(Scale Factor=&Vector2DFromCurve(
     Vector2Curve=DataInterface() {propertyValues {
       XCurve {keys [
         {arriveTangent=5.9728, interpMode=RCIM_Cubic, leaveTangent=5.9728, tangentMode=RCTM_User, time=0, value=0.1}
         {arriveTangent=-0.1, interpMode=RCIM_Cubic, leaveTangent=-0.1, tangentMode=RCTM_User, time=0.18, value=0.8}
         {interpMode=RCIM_Cubic, tangentMode=RCTM_User, time=1, value=0}
        ]}
       YCurve {keys [
         {arriveTangent=6.2715, interpMode=RCIM_Cubic, leaveTangent=6.2715, tangentMode=RCTM_User, time=0, value=0.1}
         {arriveTangent=-0.1001, interpMode=RCIM_Cubic, leaveTangent=-0.1001, tangentMode=RCTM_User, time=0.18, value=0.8}
         {interpMode=RCIM_Cubic, tangentMode=RCTM_User, time=0.9988, value=0}
        ]}
      }}
     ScaleCurve=<1 1>
    ))
   &ScaleSpriteSizeBySpeed(Max Scale Factor=<0.5 8>, Min Scale Factor=<0.8 1>)
   &ScaleColor(
    Linear Color Curve=DataInterface() {propertyValues {
      AlphaCurve {keys [<at 0 1>,<at 1 0>]}
      BlueCurve {keys [<at 0.17 1.32>,<at 0.2 0.12>,<at 0.33 0.1188>,<at 0.85 0.031>]}
      GreenCurve {keys [<at 0.17 17.16>,<at 0.2 1.56>,<at 0.33 0.4324>,<at 0.85 0.1045>]}
      RedCurve {keys [<at 0.17 33>,<at 0.2 3>,<at 0.33 2>,<at 0.85 1.5>]}
     }}
    Scale Mode=::NewEnumerator2
   )
  ]
  rendererClasses [&NiagaraSpriteRendererProperties]
  renderers [{rendererClass=&NiagaraSpriteRendererProperties}]
 }
 Speedlines {
  emitterUpdateScript [
   &EmitterState(
    Delay First Loop Only=true
    Enable Distance Culling=true
    Inactive Response=::NewEnumerator1
    Life Cycle Mode=::NewEnumerator1
    Loop Delay=1
    Loop Duration=1
    MaxDistance=2048
    Module.bUseMaxDistance=true
    Module.bUseMinDistance=false
    Scalability Mode=::NewEnumerator1
    UseLoopDelay=true
   )
   &SpawnPerUnit(
    Module.Use Max Movement Threshold=true
    Module.Use Movement Tolerance=false
    Spawn Per Unit=10
    Spawn Probability:NiagaraFloat=&Lerp_Float(A=0, B=1, Alpha=$System.FN_Timer.Value)
    Use Spawn Probability=true
   )
  ]
  particleSpawnScript [
   /Niagara/Modules/Spawn/Initialization/InitializeParticle(
    Color=&MakeLinearColorFromVectorAndFloat(
     "Vector (RGB)"=<1 1 1>
     "Float (Alpha)"=&Multiply_Float(A=&UniformRangedFloat(Minimum=0.2, Maximum=0.3), B=0.5)
    )
    Lifetime=&UniformRangedFloat(Minimum=0.1, Maximum=0.15)
    Module.Write Color=true
    Module.Write Lifetime=true
    Module.Write Mass=true
    Module.Write Material Random=true
    Module.Write Position=false
    Module.Write Scale=false
    Module.Write SpriteRotation=false
    Module.Write SpriteSize=true
    Position=&SimulationPosition()
    Sprite Size=&UniformRangedVector2D(Minimum=<4 16>, Maximum=<6 24>)
   )
   &SkeletalMeshLocation(
    Bone Sampling Mode=::NewEnumerator4
    Interpolate Mesh Sampling=false
    MaskSpawnGroup=false
    Orientation Sampling=::NewEnumerator1
    Sampled Position Offset=<0 0 -50>
    Skeletal Mesh=$User.Weapon
   )
   &SphereLocation(
    MaskSpawnGroup=false
    Offset=<0 -50 0>
    Override Seed=false
    Sphere Radius=200
    Surface Only=false
    UseNonUniformScale=false
    UseOffset=true
    UseSphereOrientation=false
   )
   &AddVelocityFromPoint(
    Module.Use Falloff Distance=false
    Origin Offset=&Add_Vector(A=$Output.SkeletalMeshLocation.SampledPositionWithoutOffset, B=<50 -50 0>)
    Velocity Strength=1000
   )
   &CameraOffset(Camera Offset Amount=-20)
  ]
  particleUpdateScript [
   &ParticleState()
   &SolveForcesAndVelocity(Clamp Velocity=false, Limit Acceleration=false)
   &ScaleSpriteSize(Scale Factor=&Vector2DFromCurve(
     Vector2Curve=DataInterface() {propertyValues {XCurve {keys [<at 0 0.5 RCIM_Cubic>,<at 0.5 1 RCIM_Cubic>]}, YCurve {keys [<at 0 0 RCIM_Cubic>,<at 0.5 1 RCIM_Cubic>]}}}
     ScaleCurve=&Vector2DFromFloat(Value=1)
    ))
   &ScaleSpriteSizeBySpeed(Max Scale Factor=<1 20>, Min Scale Factor=<0.01 0.05>, Velocity Threshold=1)
   &ScaleColor(
    Scale Alpha=&Multiply_Float(
     A=&FloatFromCurve(FloatCurve=DataInterface() {propertyValues {Curve {keys [
          <at 0 0 RCIM_Cubic>
          <at 0.3759 1.0207>
          {arriveTangent=-2.221, interpMode=RCIM_Cubic, leaveTangent=-2.221, tangentMode=RCTM_User, time=1, value=0}
         ]}}})
     B=&FloatFromCurve(
      FloatCurve=DataInterface() {propertyValues {Curve {keys [<at 0 0>,<at 1 1>]}}}
      CurveIndex=&NormalizeFloat(Value=&VectorLength(Vector=$Engine.Owner.Velocity), Min=0, Max=1)
     )
    )
    Scale RGB=&VectorFromFloat(Value=&ClampFloat(
      Float=$User.ForwardDot
      Max=&FNTimeOfDayScalar(Night Scale=1, Morning Scale=0.3, Evening Scale=0.5)
     ))
    ScaleA=true
    ScaleRGB=true
   )
  ]
  rendererClasses [&NiagaraSpriteRendererProperties]
  renderers [{rendererClass=&NiagaraSpriteRendererProperties}]
 }
 Wind_Speed_HF {
  emitterUpdateScript [
   &EmitterState(Inactive Response=::NewEnumerator1, Life Cycle Mode=::NewEnumerator1, Loop Duration=0, UseLoopDelay=false)
   &SpawnBurst_Instantaneous(
    Spawn Count=&MakeIntFromFloat(Float=&Lerp_Float(A=0, B=1, Alpha=$System.FN_Integrator001.Value))
    Use Spawn Probability=false
   )
  ]
  particleSpawnScript [
   /Niagara/Modules/Spawn/Initialization/V2/InitializeParticle(
    Color=<rgba 1 1 1 0.5>
    Lifetime=0.1
    Mesh Scale=<0.4 0.4 0.5>
    Mesh Scale Mode=::NewEnumerator4
    UsePositionOffset=true
    Write Material Random=true
   )
   &InitialMeshOrientation(Orientation Axis=<0 0 1>, Orientation Vector=<0 0 -1>, Rotation=<-0.13 0 0>, Use Orientation Vector=true, Use Rotation Vector=true)
  ]
  particleUpdateScript [
   &ParticleState()
   &SkeletalMeshLocation(
    MaskSpawnGroup=false
    Mesh Sampling Type=::NewEnumerator5
    Orientation Sampling=::NewEnumerator1
    Sampled Position Offset=<0 0 -20>
    Skeletal Mesh=$User.Weapon
    Socket Sampling Mode=::NewEnumerator5
   )
   &ScaleColor(Scale RGB=<0.5 0.75 1.5>, ScaleA=true, ScaleRGB=true)
  ]
  rendererClasses [&NiagaraMeshRendererProperties]
  renderers [{rendererClass=&NiagaraMeshRendererProperties, rendererIndex=0}]
 }
}
summary {
 systemName=NS_Glider_Parfait_Female
 emitters [
  {emitterName=Lightning_Strike_BURST, rendererClasses [&NiagaraMeshRendererProperties]}
  {emitterName=Lightning_Residue_BURST, rendererClasses [&NiagaraMeshRendererProperties]}
  {emitterName=Lightning_Sprite_IDLE, rendererClasses [&NiagaraSpriteRendererProperties]}
  {emitterName=Glow_IDLE, rendererClasses [&NiagaraLightRendererProperties, &NiagaraSpriteRendererProperties]}
  {emitterName=Lightning_Meshes_IDLE, rendererClasses [&NiagaraMeshRendererProperties]}
  {
   emitterName=Lightning_Sprite_BURST
   rendererClasses [&NiagaraSpriteRendererProperties, &NiagaraLightRendererProperties]
  }
  {emitterName=Sparks_BURST, rendererClasses [&NiagaraSpriteRendererProperties]}
  {
   emitterName=Flash_Flare_BURST
   rendererClasses [&NiagaraSpriteRendererProperties, &NiagaraLightRendererProperties]
  }
  {emitterName=Glow_Ring_BURST, rendererClasses [&NiagaraSpriteRendererProperties]}
  {emitterName=Lightning_Arc_Sprites_HF, rendererClasses [&NiagaraSpriteRendererProperties]}
  {emitterName=Wind_Speed_HF, rendererClasses [&NiagaraMeshRendererProperties]}
  {emitterName=Halo_HF, rendererClasses [&NiagaraSpriteRendererProperties]}
  {emitterName=Flash_Flare_HF, rendererClasses [&NiagaraSpriteRendererProperties]}
  {emitterName=Fissure_Glow_IDLE, rendererClasses [&NiagaraMeshRendererProperties]}
  {emitterName=Lightning_Arcs_IDLE, rendererClasses [&NiagaraMeshRendererProperties]}
  {emitterName=Lightning_Main_IDLE, rendererClasses [&NiagaraMeshRendererProperties]}
  {emitterName=Speedlines, rendererClasses [&NiagaraSpriteRendererProperties]}
  {emitterName=Lightning_Big_IDLE001, rendererClasses [&NiagaraMeshRendererProperties]}
  {
   emitterName=Lightning_Big_FORWARD
   rendererClasses [&NiagaraMeshRendererProperties, &NiagaraLightRendererProperties]
  }
  {emitterName=Lightning_Detail_FORWARD, rendererClasses [&NiagaraMeshRendererProperties]}
  {emitterName=Lightning_Main_Small_FORWARD, rendererClasses [&NiagaraMeshRendererProperties]}
  {emitterName=Lightning_Main_FORWARD, rendererClasses [&NiagaraMeshRendererProperties]}
 ]
 userVariables [
  User.bIsFrontEnd {type=NiagaraBool}
  User.bIsFrontEndPreview {type=NiagaraBool}
  User.bIsFullyDeployed {type=NiagaraBool}
  User.FadeinCurve {
   type=NiagaraDataInterfaceCurve
   description=""
   defaultValue {
    struct=NiagaraExt_VariableValue_DataInterface
    value {
     dataInterface=&self:NiagaraDataInterfaceCurve_0
     dataInterfaceClass=&NiagaraDataInterfaceCurve
    }
   }
  }
  User.FlickerSpeed {type=NiagaraFloat, defaultValue=6.666, description=""}
  User.ForwardDot {type=NiagaraFloat, defaultValue=0, description=""}
  User.IsFrontEnd {type=NiagaraBool}
  User.IsWhiteVersion {type=NiagaraBool}
  User.LightningOrder {
   type=NiagaraDataInterfaceArrayInt32
   description=""
   defaultValue {
    struct=NiagaraExt_VariableValue_DataInterface
    value {
     dataInterface=&self:NiagaraDataInterfaceArrayInt32_0
     dataInterfaceClass=&NiagaraDataInterfaceArrayInt32
    }
   }
  }
  User.LightningSpriteDepthFade {type=NiagaraFloat, defaultValue=44, description=""}
  User.RightDot {type=NiagaraFloat, defaultValue=0, description=""}
  User.Weapon {
   type=NiagaraDataInterfaceSkeletalMesh
   defaultValue {
    struct=NiagaraExt_VariableValue_DataInterface
    value {dataInterface=&self:NiagaraDataInterfaceSkeletalMesh_2}
   }
  }
 ]
}
